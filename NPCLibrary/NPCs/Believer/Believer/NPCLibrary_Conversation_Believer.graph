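# Conversation database for the Believer NPC, an alien-welcoming character, entered through the Conversation.Entry.Believer tag.
# A single speech node gives a default line or contextual greetings with curve-driven priorities, reacting to time since last match, first meeting, high Bars and a specific burger mascot outfit.
# That node branches through link nodes to shared conversation entries for basic services, NPC library quests and four Season 17 weekly conversations.
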ConversationDatabase /NPCLibrary/NPCs/Believer/Believer/NPCLibrary_Conversation_Believer
entry_points [{tag=Conversation.Entry.Believer, destinations [#1]}]
nodes [
 ConversationEntryPointNode_7 {
  id=#1
  class=ConversationEntryPointNode
  outputs [#2]
  properties {entryTag {tagName=Conversation.Entry.Believer}}
 }
 Speech_C_4 {
  id=#2
  class=Speech_C
  outputs [
   #3
   #4
   #5
   #6
   #7
   #8
  ]
  properties {
   nodeName=""
   generalConfig {
    defaultMessage="I'm telling you - if we're good to them, they'll be good to us."
    contextualMessages [
     {
      message=Are you back to welcome our new alien overlords?
      contextRequirements [{requirement {class=FortControllerRequirement_CheckTimeSinceLastPlayedMatch}}]
      priority {curve {rowName=Tandem.Speech.Priority.WhileSincePlayed}}
     }
     {
      message=Are you here to celebrate the end of the anthropocene?
      requirementMatchPolicy=RequireAny
      contextRequirements [{requirement {class=FortControllerRequirement_IsFirstEverConversationWithNPC}}]
      priority {curve {rowName=Tandem.Speech.Priority.FirstEverInteraction}}
     }
     {
      message="That's a lot of Bars. Did you know the aliens will bring forth a post-Bar society? Could be cool, right?"
      contextRequirements [{requirement {class=FortControllerRequirement_HasBars, properties {thresholdTestConfig {requirement=GreaterThanOrEqual}}}}]
      priority {curve {rowName=Tandem.Speech.Priority.HighTotalBars}}
     }
     {
      message=This guy gets it.
      contextRequirements [{requirement {
         class=FortControllerRequirement_HasCID
         properties {allowedCIds [/BRCosmetics/Athena/Items/Cosmetics/Characters/CID_505_Athena_Commando_M_MilitiaMascotBurger]}
        }}]
      priority {curve {rowName=Tandem.Speech.Priority.OutfitReaction}}
     }
    ]
   }
  }
 }
 ConversationLinkNode_5 {id=#8, class=ConversationLinkNode, properties {remoteEntryTag {tagName=Conversation.Entry.Season17.Week13}}}
 ConversationLinkNode_4 {id=#7, class=ConversationLinkNode, properties {remoteEntryTag {tagName=Conversation.Entry.Season17.Week3}}}
 ConversationLinkNode_3 {id=#6, class=ConversationLinkNode, properties {remoteEntryTag {tagName=Conversation.Entry.Season17.Week5}}}
 ConversationLinkNode_0 {id=#5, class=ConversationLinkNode, properties {remoteEntryTag {tagName=Conversation.Entry.BasicServices}}}
 ConversationLinkNode_1 {id=#4, class=ConversationLinkNode, properties {remoteEntryTag {tagName=Conversation.Entry.NPCLibrary.Quests}}}
 ConversationLinkNode_2 {id=#3, class=ConversationLinkNode, properties {remoteEntryTag {tagName=Conversation.Entry.Season17.Week1}}}
]
